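# Editor utility function library for the Animation Browser that toggles Fortnite cosmetic accessories on skeletal mesh actors in the editor world.
# It processes actors exposing an AthenaCharacter property, using a hidden_accessories actor tag as toggle state.
# Hiding rebuilds CharacterPartList, keeping the first two parts and substituting an empty local for the rest.
# Restoring clears and reassigns AthenaCharacter with change notifications, then removes the tag.
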EditorUtilityBlueprint /AnimInEdTools/AnimationBrowser/ProjectSettings/ProjectOptions/EU_AnimationBrowser_FNCosmeticFunctions_PO
graphs [
 EventGraph {links [], nodes []}
 ToggleFNCosmeticAccessories {nodes [
   ToggleFNCosmeticAccessories() {is=title K2Node_FunctionEntry}
   > n1 = GetEditorWorld(self=Get UnrealEditorSubsystem())
   > n2 = GetAllActorsOfClass!(execute, WorldContextObject:Object Reference=@n1, ActorClass:Actor Class Reference=/Script/Engine.SkeletalMeshActor) -> (then, OutActors:Array of Skeletal Mesh Actor Object References)
   > n3 = For Each Loop(Array:Array of Skeletal Mesh Actor Object References=@n2) -> (Array Element:Skeletal Mesh Actor Object Reference)
   > Set loc_SkMesh(loc_SkMesh:Skeletal Mesh Actor Object Reference=@n3) -> (Output_Get:Skeletal Mesh Actor Object Reference)
   n4 = Get loc_SkMesh() -> (loc_SkMesh:Skeletal Mesh Actor Object Reference)
   > Is Valid(InputObject=@n4)
   > n5 = GetEditorProperty!(execute:Exec, Object:Object Reference=@n4, PropertyName:Name=AthenaCharacter) -> (then:Exec, PropertyValue:Athena Character Item Definition Object Reference, ReturnValue:bool)
   > Branch(Condition=@n5.ReturnValue)
   > Set loc_Char(loc_Char:Athena Character Item Definition Object Reference=@n5) -> (Output_Get:Athena Character Item Definition Object Reference)
   > n6 = Branch(Condition=ActorHasTag!(self:Actor Object Reference=Get loc_SkMesh(), Tag:Name=hidden_accessories) -> (ReturnValue:bool))
   n7 = Get loc_SkMesh()
   > SetEditorProperty(
    Object=@n7
    PropertyName=AthenaCharacter
    PropertyValue:Athena Character Item Definition Object Reference=Get loc_Char_none() -> (loc_Char_none:Athena Character Item Definition Object Reference)
    ChangeNotifyMode=Always
   )
   > SetEditorProperty(
    Object=@n7
    PropertyName=AthenaCharacter
    PropertyValue=Get loc_Char() -> (loc_Char:Athena Character Item Definition Object Reference)
    ChangeNotifyMode=Always
   )
   > Remove Item!(
    execute
    TargetArray:Array of Names=Get Tags!(self:Actor Object Reference=Get loc_SkMesh()) -> (Tags:Array of Names)
    Item:"Name (by ref)"=hidden_accessories
   ) -> (then, ReturnValue:bool)
   n8 = GetEditorProperty(execute=@n6.else, Object=Get loc_SkMesh(), PropertyName=CharacterPartList) -> (PropertyValue:Array of Custom Character Part Object References)
   > n9 = Branch(Condition=@n8.ReturnValue)
   > n10 = For Each Loop(Array:Array of Custom Character Part Object References=@n8) -> (Array Element:Custom Character Part Object Reference)
   Add Unique(execute=@n10.Completed, TargetArray:Array of Names=Get Tags(self=Get loc_SkMesh()), NewItem:"Name (by ref)"=hidden_accessories)
   > n11 = SetEditorProperty(
    Object=Get loc_SkMesh()
    PropertyName=CharacterPartList
    PropertyValue:Array of Custom Character Part Object References=Get loc_charAccSlots_new() -> (loc_charAccSlots_new:Array of Custom Character Part Object References)
   )
   > n12 = Branch(Condition=@n11)
   Remove Item(execute=@n12.else, TargetArray=Get Tags(self=Get loc_SkMesh()), Item=hidden_accessories)
   n13 = Branch(execute=@n10.LoopBody, Condition="integer > integer"(A=@n10.Array Index, B=1))
   > Add(
    TargetArray:Array of Custom Character Part Object References=Get loc_charAccSlots_new()
    NewItem:"Custom Character Part Object Reference (by ref)"=Get loc_customChar_none() -> (loc_customChar_none:Custom Character Part Object Reference)
   ) {is=title K2Node_CallArrayFunction}
   Add(execute=@n13.else, TargetArray=Get loc_charAccSlots_new(), NewItem=@n10)
   PrintString(
    execute=@n9.else
    InString="Warning: [ToggleFNCosmeticAccessories] Failed to find named property CharacterPartList on SK Actor, unable to hide accessories"
    bPrintToScreen=false
   )
  ]}
]
variables []
